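; Quest interactable tiny rubber duck prop that quacks as a location hint and squeaks when searched.
; A looping timer started on quest initialization plays the hint; objective completion clears it and plays a delayed collected sound.
; Quest updates disable collision, hide the prop and stop the timer when the local player's completed-objective count is zero or more.
; The construction script creates a dynamic material that sets the duck's bob speed.
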(event EventQuestInitializedAndValid
  ;[1 1 1] Set up Looping Hint Quack Sound
  (|CallParentFunction)
  (bind _returnvalue (Utilities|Time|SetTimerbyFunctionName 0 "PlayIntermittentHintSound" (Variables|Default|GetIntermittentSoundLoopTime) true))
  (Variables|Default|SetIntermittentHitSoundTimer _returnvalue))

(event Interaction|EventBlueprintOnBeginInteract
  (params (InteractingPawn :type Fort%20Pawn%20Object%20Reference) (InteractionBeingAttempted :type EInteractionBeingAttempted%20Enum :default "FirstInteraction") (InteractionType :type TInteractionType%20Enum :default "IT_NoInteraction"))
  ;[1 1 1] Play Searching Squeak Sound
  (|CallParentFunction 0 "FirstInteraction" "IT_NoInteraction")
  (Audio|PlaySoundatLocation (Variables|Default|GetDuckSearchingSound) (Transformation|GetActorLocation)))

(event EventThisQuestObjectiveComplete
  (params (IneractingPawn :type Fort%20Pawn%20Object%20Reference))
  ;[1 1 1] Stop Hint Sound and Play Collected Duck sound
  (|CallParentFunction 0)
  (Utilities|Time|ClearandInvalidateTimerbyHandle (Variables|Default|GetIntermittentHitSoundTimer))
  (Utilities|FlowControl|Delay 0.2)
  (Audio|PlaySoundatLocation (Variables|Default|GetDuckCollectedSound) (Transformation|GetActorLocation)))

(event EventQuestUpdatedFromAnywhere
  ;[1 1 1] Check if Quest Complete after collection, if so hide the rest
  (|CallParentFunction)
  ; S09 Extraction - Removed Errors
  (bind _returnvalue (Quest|GetNumCompletedObjectives (Quests|GetQuestManager (Utilities|Array|Get(acopy) (Fortnite|GetLocalFortPlayerControllers true)) "Invalid") (Variables|QuestStuff|GetQuestItem)))
  (if (>= _returnvalue 0)
    (Collision|SetActorEnableCollision false)
    (CallFunction|HideQuestProp)
    (Utilities|Time|ClearandInvalidateTimerbyHandle (Variables|Default|GetIntermittentHitSoundTimer))))

(fn PlayIntermittentHintSound
  (Audio|PlaySoundatLocation (Variables|Default|GetIntermittentHintSound) (Transformation|GetActorLocation)))

(fn ConstructionScript
  (|CallParentFunction)
  (bind _returnvalue (Rendering|Material|CreateDynamicMaterialInstance (Variables|Default|GetSM_Duck) 0 "/Figment_S09/Athena/Items/QuestInteractables/RubberDuckies/FX/M_RubberDuck_QuestInteractable.M_RubberDuck_QuestInteractable" "None" :!class /Script/Engine.PrimitiveComponent))
  (bind _output_get (Variables|Default|SetDuckMat _returnvalue))
  (Rendering|Material|SetScalarParameterValue _output_get "BobSpeed" (Variables|Default|GetBobSpeed) :!class /Script/Engine.MaterialInstanceDynamic))
